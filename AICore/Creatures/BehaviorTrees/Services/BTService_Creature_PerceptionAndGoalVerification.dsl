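; Behavior tree service that decides when a creature has verified its GoalActor as a threat, supporting stealth.
; Alert level changes track seen time, deducting time out of sight, and reset everything on Unaware.
; Hurt or Touched verifies instantly; while seeing, tick verifies by blackboard-set duration or proximity.
; Verification applies a GameplayEffect, whose granted tags mark the state, and timestamps the blackboard for a reinforcement timeout.
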(event EventReceiveDeactivation
  (params (OwnerActor :type Actor%20Object%20Reference))
  ; Deactivation and cleanup
  (Variables|Default|SetValidBlackboard 0)
  (Variables|InternalLogicVariables|SetControlledPawn 0)
  (bind _validalertlevelcomponent (Variables|Default|GetValidAlertLevelComponent)
    (:then
      (AlertLevelComponent|UnbindEventfromOnAlertLevelDynamicChangeEvent _validalertlevelcomponent (delegate Custom|OnAlertLevelDynamicChangeEvent_Event))
      (Variables|Default|SetValidAlertLevelComponent 0))
    (:else)))

(event Custom|OnAlertLevelDynamicChangeEvent_Event
  (params (PrevAlertLevel :type EAlertLevel%20Enum :default "Unaware") (NewAlertLevel :type EAlertLevel%20Enum :default "Unaware"))
  (bind _goalactor (CallFunction|GetGoalActorBBKey))
  ; Event - OnAlertLevelChanged
  ; Check AlertLevel
  (Variables|InternalLogicVariables|SetCurrentAlertLevel NewAlertLevel)
  (bind _validalertlevelcomponent (Variables|Default|GetValidAlertLevelComponent)
    (:then
      (Variables|InternalLogicVariables|SetLKP%20Type (AlertLevelComponent|GetLastKnownPositionType _validalertlevelcomponent))
      (Utilities|FlowControl|Sequence
        (:then_0
          ; Track Acquisition and Loss of Sight in order to Set GoalActorVerified = True
          ; Hurt or Touched will immediately Set GoalActorVerified = True
          (switch Utilities|FlowControl|Switch|SwitchonEAlertLevel (Variables|InternalLogicVariables|GetCurrentAlertLevel)
            (:Unaware
              (Utilities|FlowControl|Sequence
                (:then_0
                  ;; Clear/Reset when we go Unaware again
                  (Variables|InternalLogicVariables|SetHasSeenSinceLastUnaware false)
                  (Variables|InternalLogicVariables|SetLastSeeingTime 0.0)
                  (Variables|InternalLogicVariables|SetLostSeeingTime 0.0)
                  (Variables|InternalLogicVariables|SetCurrentSeeingDuration 0.0)
                  (Variables|InternalLogicVariables|SetTotalSeeingDurationSinceUnaware 0.0)
                  (Variables|InternalLogicVariables|SetGameTimeReferencePointForSeeingDurationComparison 0.0)
                  (CallFunction|SetGoalVerificationStatus false))
                (:then_1
                  _)))
            (:Suspicious
              _)
            (:LostTarget
              _)
            (:Threatened
              ;; Track latest Seeing start time, and adjust our Time Reference Point if we have previous Seeing Time remaining
              (if (CallFunction|LKPIsHurtorTouched)
                (CallFunction|SetGoalVerificationStatus true :!pinned true) ; Set Goal Actor Verification to True\nInstant Failure of Stealth
                (elif (CallFunction|GetPerceptionGoalActorIsBeingSeenBBKey)
                  (Utilities|FlowControl|Sequence
                    (:then_0
                      (Variables|InternalLogicVariables|SetIsSeeing true)
                      (if (Variables|InternalLogicVariables|GetHasSeenSinceLastUnaware)
                        (Variables|InternalLogicVariables|SetLastSeeingTime (Utilities|Time|GetGameTimeinSeconds)) ; Track when we started Seeing
                        (Variables|InternalLogicVariables|SetTotalSeeingDurationSinceUnaware (Math|Float|Max(Float) (- (Variables|InternalLogicVariables|GetTotalSeeingDurationSinceUnaware) (- (Utilities|Time|GetGameTimeinSeconds) (Variables|InternalLogicVariables|GetLostSeeingTime))))) ; Subtract any "lost" Seeing time from our current total.
                        (Variables|InternalLogicVariables|SetGameTimeReferencePointForSeeingDurationComparison (- (Utilities|Time|GetGameTimeinSeconds) (Variables|InternalLogicVariables|GetTotalSeeingDurationSinceUnaware)))
                        (else
                          (bind _output_get (Variables|InternalLogicVariables|SetLastSeeingTime (Utilities|Time|GetGameTimeinSeconds))) ; Just set our baseline.
                          (Variables|InternalLogicVariables|SetGameTimeReferencePointForSeeingDurationComparison _output_get))))
                    (:then_1
                      (Variables|InternalLogicVariables|SetHasSeenSinceLastUnaware true))
                    (:then_2
                      ;; Can we Verify our GoalActor by Seeing it over time or by Distance?
                      (bind _goalisverified (Default|IsGoalVerified?))
                      (if (not _goalisverified)
                        (if (> (CallFunction|GetSeeingDurationRequiredforGoalVerificationBBKey) 0)
                          (if (CallFunction|GetPerceptionGoalActorIsBeingSeenBBKey)
                            (if (>= (- (Utilities|Time|GetGameTimeinSeconds) (Variables|InternalLogicVariables|GetGameTimeReferencePointForSeeingDurationComparison)) (CallFunction|GetSeeingDurationRequiredforGoalVerificationBBKey))
                              (CallFunction|SetGoalVerificationStatus true :!pinned true) ; Verify GoalActor!\nBased on Time
                              (elif (> (CallFunction|GetGoalDistanceforInstantGoalVerificationBBKey) 0)
                                (Utilities|IsValid _goalactor
                                  (:Is%20Valid
                                    (bind _asactor (Utilities|Casting|CastToActor _goalactor))
                                    (if (< (Transformation|GetDistanceTo _asactor (Variables|InternalLogicVariables|GetControlledPawn)) (CallFunction|GetGoalDistanceforInstantGoalVerificationBBKey))
                                      (CallFunction|SetGoalVerificationStatus true :!pinned true) ; Verify GoalActor!\nBased on Distance
                                    ))))))
                          (else
                            _)))))
                  (elif (CallFunction|LKPIsSeeingorSeen)
                    _)))))
          ;; Track elapsed time when "Un-Seeing"
          (if (Variables|InternalLogicVariables|GetIsSeeing)
            (Variables|InternalLogicVariables|SetIsSeeing false)
            (Variables|InternalLogicVariables|SetLostSeeingTime (Utilities|Time|GetGameTimeinSeconds))
            (bind _output_get_2 (Variables|InternalLogicVariables|SetCurrentSeeingDuration (- (Variables|InternalLogicVariables|GetLostSeeingTime) (Variables|InternalLogicVariables|GetLastSeeingTime)))) ; How much Time Seeing did we get?
            (Variables|InternalLogicVariables|SetTotalSeeingDurationSinceUnaware (+ _output_get_2 (Variables|InternalLogicVariables|GetTotalSeeingDurationSinceUnaware))) ; Total Seeing Time since last Unaware?
          ))
        (:then_1
          ; Debug Seeing?
          (Variables|Default|GetValidBlackboard
            (:then
              (if (Variables|Setup|GetDebug)
                (Development|PrintString (select (Variables|InternalLogicVariables|GetHasSeenSinceLastUnaware) (Utilities|String|BuildString(Float) (Utilities|String|BuildString(Float) (Utilities|String|BuildString(Name) (Utilities|String|Append (Utilities|String|SelectString (Utilities|String|SelectString "Threatened and SEEING" "Threatened NOT Seeing" (Utilities|Enum|Equal(Enum) (AlertLevelComponent|GetLastKnownPositionType (Variables|Default|GetValidAlertLevelComponent)) "Seeing")) "NOT Threatened (UnSee)" (Utilities|Enum|Equal(Enum) (Variables|InternalLogicVariables|GetCurrentAlertLevel) "Threatened")) " | ") "LKP Type = " (Utilities|Name|EnumtoName (AlertLevelComponent|GetLastKnownPositionType (Variables|Default|GetValidAlertLevelComponent))) " | ") "TOTAL SEEING TIME = " (Variables|InternalLogicVariables|GetTotalSeeingDurationSinceUnaware) " | ") "Time Left = " (- (+ (Variables|InternalLogicVariables|GetGameTimeReferencePointForSeeingDurationComparison) (CallFunction|GetSeeingDurationRequiredforGoalVerificationBBKey)) (Utilities|Time|GetGameTimeinSeconds))) (Utilities|String|Append (Utilities|String|BuildString(Float) (Utilities|String|BuildString(Name) (Utilities|String|Append (Utilities|String|SelectString (Utilities|String|SelectString "Threatened and SEEING" "Threatened NOT Seeing" (Utilities|Enum|Equal(Enum) (AlertLevelComponent|GetLastKnownPositionType (Variables|Default|GetValidAlertLevelComponent)) "Seeing")) "NOT Threatened (UnSee)" (Utilities|Enum|Equal(Enum) (Variables|InternalLogicVariables|GetCurrentAlertLevel) "Threatened")) " | ") "LKP Type = " (Utilities|Name|EnumtoName (AlertLevelComponent|GetLastKnownPositionType (Variables|Default|GetValidAlertLevelComponent))) " | ") "TOTAL SEEING TIME = " (Variables|InternalLogicVariables|GetTotalSeeingDurationSinceUnaware) " | ") "Time Left = XXXX")))))
            (:else)))))
    (:else)))

(event EventReceiveActivation
  (params (OwnerActor :type Actor%20Object%20Reference))
  (bind _goalactor (CallFunction|GetGoalActorBBKey))
  ; Event - Activation
  ; Check AlertLevel
  (Utilities|VariableSetup
    (:Success
      (bind _validalertlevelcomponent (Variables|Default|GetValidAlertLevelComponent)
        (:then
          (AlertLevelComponent|AssignOnAlertLevelDynamicChangeEvent _validalertlevelcomponent (delegate Custom|OnAlertLevelDynamicChangeEvent_Event))
          (Variables|InternalLogicVariables|SetCurrentAlertLevel (AlertLevelComponent|GetCurrentAlertLevel _validalertlevelcomponent))
          (bind _validalertlevelcomponent_2 (Variables|Default|GetValidAlertLevelComponent)
            (:then
              (Variables|InternalLogicVariables|SetLKP%20Type (AlertLevelComponent|GetLastKnownPositionType _validalertlevelcomponent_2))
              (Utilities|FlowControl|Sequence
                (:then_0
                  ; Track Acquisition and Loss of Sight in order to Set GoalActorVerified = True
                  ; Hurt or Touched will immediately Set GoalActorVerified = True
                  (switch Utilities|FlowControl|Switch|SwitchonEAlertLevel (Variables|InternalLogicVariables|GetCurrentAlertLevel)
                    (:Unaware
                      (Utilities|FlowControl|Sequence
                        (:then_0
                          ;; Clear/Reset when we go Unaware again
                          (Variables|InternalLogicVariables|SetHasSeenSinceLastUnaware false)
                          (Variables|InternalLogicVariables|SetLastSeeingTime 0.0)
                          (Variables|InternalLogicVariables|SetLostSeeingTime 0.0)
                          (Variables|InternalLogicVariables|SetCurrentSeeingDuration 0.0)
                          (Variables|InternalLogicVariables|SetTotalSeeingDurationSinceUnaware 0.0)
                          (Variables|InternalLogicVariables|SetGameTimeReferencePointForSeeingDurationComparison 0.0)
                          (CallFunction|SetGoalVerificationStatus false))
                        (:then_1
                          _)))
                    (:Suspicious
                      _)
                    (:LostTarget
                      _)
                    (:Threatened
                      ;; Track latest Seeing start time, and adjust our Time Reference Point if we have previous Seeing Time remaining
                      (if (CallFunction|LKPIsHurtorTouched)
                        (CallFunction|SetGoalVerificationStatus true :!pinned true) ; Set Goal Actor Verification to True\nInstant Failure of Stealth
                        (elif (CallFunction|GetPerceptionGoalActorIsBeingSeenBBKey)
                          (Utilities|FlowControl|Sequence
                            (:then_0
                              (Variables|InternalLogicVariables|SetIsSeeing true)
                              (if (Variables|InternalLogicVariables|GetHasSeenSinceLastUnaware)
                                (Variables|InternalLogicVariables|SetLastSeeingTime (Utilities|Time|GetGameTimeinSeconds)) ; Track when we started Seeing
                                (Variables|InternalLogicVariables|SetTotalSeeingDurationSinceUnaware (Math|Float|Max(Float) (- (Variables|InternalLogicVariables|GetTotalSeeingDurationSinceUnaware) (- (Utilities|Time|GetGameTimeinSeconds) (Variables|InternalLogicVariables|GetLostSeeingTime))))) ; Subtract any "lost" Seeing time from our current total.
                                (Variables|InternalLogicVariables|SetGameTimeReferencePointForSeeingDurationComparison (- (Utilities|Time|GetGameTimeinSeconds) (Variables|InternalLogicVariables|GetTotalSeeingDurationSinceUnaware)))
                                (else
                                  (bind _output_get (Variables|InternalLogicVariables|SetLastSeeingTime (Utilities|Time|GetGameTimeinSeconds))) ; Just set our baseline.
                                  (Variables|InternalLogicVariables|SetGameTimeReferencePointForSeeingDurationComparison _output_get))))
                            (:then_1
                              (Variables|InternalLogicVariables|SetHasSeenSinceLastUnaware true))
                            (:then_2
                              ;; Can we Verify our GoalActor by Seeing it over time or by Distance?
                              (bind _goalisverified (Default|IsGoalVerified?))
                              (if (not _goalisverified)
                                (if (> (CallFunction|GetSeeingDurationRequiredforGoalVerificationBBKey) 0)
                                  (if (CallFunction|GetPerceptionGoalActorIsBeingSeenBBKey)
                                    (if (>= (- (Utilities|Time|GetGameTimeinSeconds) (Variables|InternalLogicVariables|GetGameTimeReferencePointForSeeingDurationComparison)) (CallFunction|GetSeeingDurationRequiredforGoalVerificationBBKey))
                                      (CallFunction|SetGoalVerificationStatus true :!pinned true) ; Verify GoalActor!\nBased on Time
                                      (elif (> (CallFunction|GetGoalDistanceforInstantGoalVerificationBBKey) 0)
                                        (Utilities|IsValid _goalactor
                                          (:Is%20Valid
                                            (bind _asactor (Utilities|Casting|CastToActor _goalactor))
                                            (if (< (Transformation|GetDistanceTo _asactor (Variables|InternalLogicVariables|GetControlledPawn)) (CallFunction|GetGoalDistanceforInstantGoalVerificationBBKey))
                                              (CallFunction|SetGoalVerificationStatus true :!pinned true) ; Verify GoalActor!\nBased on Distance
                                            ))))))
                                  (else
                                    _)))))
                          (elif (CallFunction|LKPIsSeeingorSeen)
                            _)))))
                  ;; Track elapsed time when "Un-Seeing"
                  (if (Variables|InternalLogicVariables|GetIsSeeing)
                    (Variables|InternalLogicVariables|SetIsSeeing false)
                    (Variables|InternalLogicVariables|SetLostSeeingTime (Utilities|Time|GetGameTimeinSeconds))
                    (bind _output_get_2 (Variables|InternalLogicVariables|SetCurrentSeeingDuration (- (Variables|InternalLogicVariables|GetLostSeeingTime) (Variables|InternalLogicVariables|GetLastSeeingTime)))) ; How much Time Seeing did we get?
                    (Variables|InternalLogicVariables|SetTotalSeeingDurationSinceUnaware (+ _output_get_2 (Variables|InternalLogicVariables|GetTotalSeeingDurationSinceUnaware))) ; Total Seeing Time since last Unaware?
                  ))
                (:then_1
                  ; Debug Seeing?
                  (Variables|Default|GetValidBlackboard
                    (:then
                      (if (Variables|Setup|GetDebug)
                        (Development|PrintString (select (Variables|InternalLogicVariables|GetHasSeenSinceLastUnaware) (Utilities|String|BuildString(Float) (Utilities|String|BuildString(Float) (Utilities|String|BuildString(Name) (Utilities|String|Append (Utilities|String|SelectString (Utilities|String|SelectString "Threatened and SEEING" "Threatened NOT Seeing" (Utilities|Enum|Equal(Enum) (AlertLevelComponent|GetLastKnownPositionType (Variables|Default|GetValidAlertLevelComponent)) "Seeing")) "NOT Threatened (UnSee)" (Utilities|Enum|Equal(Enum) (Variables|InternalLogicVariables|GetCurrentAlertLevel) "Threatened")) " | ") "LKP Type = " (Utilities|Name|EnumtoName (AlertLevelComponent|GetLastKnownPositionType (Variables|Default|GetValidAlertLevelComponent))) " | ") "TOTAL SEEING TIME = " (Variables|InternalLogicVariables|GetTotalSeeingDurationSinceUnaware) " | ") "Time Left = " (- (+ (Variables|InternalLogicVariables|GetGameTimeReferencePointForSeeingDurationComparison) (CallFunction|GetSeeingDurationRequiredforGoalVerificationBBKey)) (Utilities|Time|GetGameTimeinSeconds))) (Utilities|String|Append (Utilities|String|BuildString(Float) (Utilities|String|BuildString(Name) (Utilities|String|Append (Utilities|String|SelectString (Utilities|String|SelectString "Threatened and SEEING" "Threatened NOT Seeing" (Utilities|Enum|Equal(Enum) (AlertLevelComponent|GetLastKnownPositionType (Variables|Default|GetValidAlertLevelComponent)) "Seeing")) "NOT Threatened (UnSee)" (Utilities|Enum|Equal(Enum) (Variables|InternalLogicVariables|GetCurrentAlertLevel) "Threatened")) " | ") "LKP Type = " (Utilities|Name|EnumtoName (AlertLevelComponent|GetLastKnownPositionType (Variables|Default|GetValidAlertLevelComponent))) " | ") "TOTAL SEEING TIME = " (Variables|InternalLogicVariables|GetTotalSeeingDurationSinceUnaware) " | ") "Time Left = XXXX")))))
                    (:else)))))
            (:else)))
        (:else)))))

(event EventReceiveTick
  (params (OwnerActor :type Actor%20Object%20Reference) (DeltaSeconds :type Float%20%28single-precision%29))
  (bind _goalactor (CallFunction|GetGoalActorBBKey))
  ; Event Tick - Only check our Time Accumulation and don't set any goalposts.
  ; Must be Seeing to test for Goal Verification
  (Utilities|VariableSetup
    (:Success
      (bind _validalertlevelcomponent (Variables|Default|GetValidAlertLevelComponent)
        (:then
          (Variables|InternalLogicVariables|SetCurrentAlertLevel (AlertLevelComponent|GetCurrentAlertLevel _validalertlevelcomponent))
          (Variables|InternalLogicVariables|SetLKP%20Type (AlertLevelComponent|GetLastKnownPositionType _validalertlevelcomponent))
          (Utilities|FlowControl|Sequence
            (:then_0
              (if (CallFunction|LKPIsHurtorTouched)
                (CallFunction|SetGoalVerificationStatus true :!pinned true) ; Set Goal Actor Verification to True\nInstant Failure of Stealth
              ))
            (:then_1
              (if (Variables|InternalLogicVariables|GetIsSeeing)
                (Utilities|FlowControl|Sequence
                  (:then_0
                    ; Track Acquisition and Loss of Sight in order to Set GoalActorVerified = True
                    ; Hurt or Touched will immediately Set GoalActorVerified = True
                    ;; Can we Verify our GoalActor by Seeing it over time or by Distance?
                    (bind _goalisverified (Default|IsGoalVerified?))
                    (if (not _goalisverified)
                      (if (> (CallFunction|GetSeeingDurationRequiredforGoalVerificationBBKey) 0)
                        (if (CallFunction|GetPerceptionGoalActorIsBeingSeenBBKey)
                          (if (>= (- (Utilities|Time|GetGameTimeinSeconds) (Variables|InternalLogicVariables|GetGameTimeReferencePointForSeeingDurationComparison)) (CallFunction|GetSeeingDurationRequiredforGoalVerificationBBKey))
                            (CallFunction|SetGoalVerificationStatus true :!pinned true) ; Verify GoalActor!\nBased on Time
                            (elif (> (CallFunction|GetGoalDistanceforInstantGoalVerificationBBKey) 0)
                              (Utilities|IsValid _goalactor
                                (:Is%20Valid
                                  (bind _asactor (Utilities|Casting|CastToActor _goalactor))
                                  (if (< (Transformation|GetDistanceTo _asactor (Variables|InternalLogicVariables|GetControlledPawn)) (CallFunction|GetGoalDistanceforInstantGoalVerificationBBKey))
                                    (CallFunction|SetGoalVerificationStatus true :!pinned true) ; Verify GoalActor!\nBased on Distance
                                  ))))))
                        (else
                          _))))
                  (:then_1
                    ; Debug Seeing?
                    (Variables|Default|GetValidBlackboard
                      (:then
                        (if (Variables|Setup|GetDebug)
                          (Development|PrintString (select (Variables|InternalLogicVariables|GetHasSeenSinceLastUnaware) (Utilities|String|BuildString(Float) (Utilities|String|BuildString(Float) (Utilities|String|BuildString(Name) (Utilities|String|Append (Utilities|String|SelectString (Utilities|String|SelectString "Threatened and SEEING" "Threatened NOT Seeing" (Utilities|Enum|Equal(Enum) (AlertLevelComponent|GetLastKnownPositionType (Variables|Default|GetValidAlertLevelComponent)) "Seeing")) "NOT Threatened (UnSee)" (Utilities|Enum|Equal(Enum) (Variables|InternalLogicVariables|GetCurrentAlertLevel) "Threatened")) " | ") "LKP Type = " (Utilities|Name|EnumtoName (AlertLevelComponent|GetLastKnownPositionType (Variables|Default|GetValidAlertLevelComponent))) " | ") "TOTAL SEEING TIME = " (Variables|InternalLogicVariables|GetTotalSeeingDurationSinceUnaware) " | ") "Time Left = " (- (+ (Variables|InternalLogicVariables|GetGameTimeReferencePointForSeeingDurationComparison) (CallFunction|GetSeeingDurationRequiredforGoalVerificationBBKey)) (Utilities|Time|GetGameTimeinSeconds))) (Utilities|String|Append (Utilities|String|BuildString(Float) (Utilities|String|BuildString(Name) (Utilities|String|Append (Utilities|String|SelectString (Utilities|String|SelectString "Threatened and SEEING" "Threatened NOT Seeing" (Utilities|Enum|Equal(Enum) (AlertLevelComponent|GetLastKnownPositionType (Variables|Default|GetValidAlertLevelComponent)) "Seeing")) "NOT Threatened (UnSee)" (Utilities|Enum|Equal(Enum) (Variables|InternalLogicVariables|GetCurrentAlertLevel) "Threatened")) " | ") "LKP Type = " (Utilities|Name|EnumtoName (AlertLevelComponent|GetLastKnownPositionType (Variables|Default|GetValidAlertLevelComponent))) " | ") "TOTAL SEEING TIME = " (Variables|InternalLogicVariables|GetTotalSeeingDurationSinceUnaware) " | ") "Time Left = XXXX")))))
                      (:else))))))))
        (:else)))))

(fn GetGoalActorBBKey
  (return (AI|Components|Blackboard|GetValueasObject (Variables|Default|GetValidBlackboard) (Utilities|Name|MakeLiteralName "GoalActor"))))

(fn GetGoalDistanceForInstantGoalVerificationBBKey
  (return (AI|Components|Blackboard|GetValueasFloat (Variables|Default|GetValidBlackboard) (Utilities|Name|MakeLiteralName "Investigation_GoalDistanceForInstantGoalVerification"))))

(fn GetPerceptionGoalActorIsBeingSeenBBKey
  (return (AI|Components|Blackboard|GetValueasBool (Variables|Default|GetValidBlackboard) (Utilities|Name|MakeLiteralName "PerceptionGoalActorIsbeingSeen"))))

(fn GetSeeingDurationRequiredForGoalVerificationBBKey
  (return (AI|Components|Blackboard|GetValueasFloat (Variables|Default|GetValidBlackboard) (Utilities|Name|MakeLiteralName "Investigation_SeeingDurationRequiredForGoalVerification"))))

(fn IsGoalVerified?
  (bind _ge_creature_goalverified (Variables|Setup|GetGE_Creature_GoalVerified))
  (bind _returnvalue (Ability|GameplayEffect|GetGrantedTags _ge_creature_goalverified :!class /Script/GameplayAbilities.AbilitySystemBlueprintLibrary))
  ; Do we have the Verified Tags on the Pawn?
  (bind _controlledpawn (Variables|InternalLogicVariables|GetControlledPawn)
    (:then
      (if (Utilities|GameplayTagContainerNotEmpty _returnvalue)
        (bind _returnvalue_2 (GameplayTags|HasAllMatchingGameplayTags(Message) _controlledpawn _returnvalue)) ; Has Verified Tags from our GE Applied?
        (if _returnvalue_2
          (return true)
          (else
            (return false)))))
    (:else)))

(fn LKPIsHurtOrTouched
  (return (or (Utilities|Enum|Equal(Enum) (AlertLevelComponent|GetLastKnownPositionType (Variables|Default|GetValidAlertLevelComponent)) "Hurt") (Utilities|Enum|Equal(Enum) (AlertLevelComponent|GetLastKnownPositionType (Variables|Default|GetValidAlertLevelComponent)) "Touched"))))

(fn LKPIsSeeingOrSeen
  (return (or (Utilities|Enum|Equal(Enum) (AlertLevelComponent|GetLastKnownPositionType (Variables|Default|GetValidAlertLevelComponent)) "Seeing") (Utilities|Enum|Equal(Enum) (AlertLevelComponent|GetLastKnownPositionType (Variables|Default|GetValidAlertLevelComponent)) "Seen"))))

(fn SetGoalVerificationStatus
  (params (SetGoalActorVerified? :type Boolean))
  ; Do we actually KNOW what our Goal Actor IS?
  ;
  ; Used for Stealthy moments where we want Perception over Time
  ; to lead us to understand that our Goal Actor is actually a Threat.
  (bind _controlledpawn (Variables|InternalLogicVariables|GetControlledPawn)
    (:then
      (Utilities|IsValid (Ability|GetAbilitySystemComponent _controlledpawn)
        (:Is%20Valid
          (if (Variables|GetSetGoalActorVerified?)
            (bind _returnvalue (Ability|GameplayEffect|IsActiveGameplayEffectHandleValid (Variables|InternalLogicVariables|GetAppliedGoalVerifiedGE)))
            (if (not _returnvalue)
              (bind _returnvalue_2 (GameplayEffects|ApplyGameplayEffectToTarget :self (Ability|GetAbilitySystemComponent (Variables|InternalLogicVariables|GetControlledPawn)) :GameplayEffectClass (Variables|Setup|GetGE_Creature_GoalVerified) :Target (Ability|GetAbilitySystemComponent (Variables|InternalLogicVariables|GetControlledPawn)) :Level 0.0 :!pinned true)) ; Goal Verified!
              (Variables|InternalLogicVariables|SetAppliedGoalVerifiedGE _returnvalue_2)
              (Variables|Default|GetValidBlackboard
                (:then
                  (CallFunction|SetGoalVerifiedTimestampBBKey (Utilities|Time|GetGameTimeinSeconds)) ; Set Timestamp of when we Verfied our Goal so that we can\ndo a timeout test against it and allow summoning reinforcements\neven if we can't reach our reinforcement location.\n\nIn other words, don't allow the AI to get stuck and never spawn\nthe reinforcements.
                )
                (:else)))
            (else
              (bind _returnvalue_3 (Ability|GameplayEffect|IsActiveGameplayEffectHandleValid (Variables|InternalLogicVariables|GetAppliedGoalVerifiedGE)))
              (if _returnvalue_3
                (GameplayEffects|RemoveActiveGameplayEffect (Ability|GetAbilitySystemComponent (Variables|InternalLogicVariables|GetControlledPawn)) (Variables|InternalLogicVariables|GetAppliedGoalVerifiedGE) -1 :!pinned true) ; Remove Goal Verification
                (Variables|InternalLogicVariables|SetAppliedGoalVerifiedGE)
                (Variables|Default|GetValidBlackboard
                  (:then
                    (CallFunction|SetGoalVerifiedTimestampBBKey 0.0) ; Reset to 0
                  )
                  (:else))))))))
    (:else)))

(fn SetGoalVerifiedTimestampBBKey
  (params (GoalVerificationTime :type Float%20%28double-precision%29))
  (bind _validblackboard (Variables|Default|GetValidBlackboard)
    (:then
      (AI|Components|Blackboard|SetValueasFloat _validblackboard (Utilities|Name|MakeLiteralName "Investigation_GoalVerifiedTimestamp") (Variables|GetGoalVerificationTime)))
    (:else)))
